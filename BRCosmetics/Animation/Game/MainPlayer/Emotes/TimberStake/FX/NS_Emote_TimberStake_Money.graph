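# Niagara emote effect for TimberStake that throws tinted V-Bucks mesh particles from the gadget's mainSocket.
# User.Burst scales an instant burst, gated in gameplay by smoothed owner movement, with separate front-end tuning for count and launch.
# Meshes pop in, tumble under gravity, curl noise and drag with platform-gated collision, then shrink out.
# Particles inherit owner movement above a speed threshold; alternate spawners and the second collision are disabled.
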NiagaraSystem /BRCosmetics/Animation/Game/MainPlayer/Emotes/TimberStake/FX/NS_Emote_TimberStake_Money
dependencies {
 usedDataInterfaces [
  &NiagaraDataInterfaceSocketReader
  &NiagaraDataInterfaceArrayColor
  &NiagaraDataInterfaceCurve
  &NiagaraDataInterfacePlatformSet
 ]
 usedDynamicInputs [&RandomRangeVector]
 usedModules [&SystemState]
}
emitters {VBucks {
  emitterUpdateScript [
   &self:InterpolateSpeedThresholdEmitter_01(
    Define Initial Value=false
    Rate Of Change=5
    Target Float Value:NiagaraFloat=&MakeCustomFloatFromBool(
     Bool=&SetBoolByFloatComparison(
      A=&Add_Float(
       A=&Add_Float(
        A=&Abs_Float(Float:NiagaraFloat=&MakeFloatFromVector(Vector=$Engine.Owner.Velocity, Channel=::NewEnumerator0))
        B=&Abs_Float(Float:NiagaraFloat=&MakeFloatFromVector(Vector=$Engine.Owner.Velocity, Channel=::NewEnumerator2))
       )
       B=&Abs_Float(Float:NiagaraFloat=&MakeFloatFromVector(Vector=$Engine.Owner.Velocity, Channel=::NewEnumerator1))
      )
      B=46.278
     )
     True Float=1
     False Float=0
    )
   ) {module=InterpolateSpeedThresholdEmitter_01}
   &EmitterState()
   &SpawnBurst_Instantaneous(
    Spawn Count=&MakeIntFromFloat(Float=&FN_FrontEndFloatModifier(
      Gameplay=&Lerp_Float(
       A=0
       B=&Multiply_Float(A=$User.Burst, B=&RandomRangeFloat(Minimum=8, Maximum=12))
       Alpha=$Emitter.InterpolateSpeedThresholdEmitter_01.InterpolateOverTime_Float.Moving Average
      )
      Front End=&Multiply_Float(A=$User.Burst, B=&RandomRangeFloat(Minimum=6, Maximum=10))
     ))
    Spawn Probability=0.7
    Use Spawn Probability=false
    UseLoopCountLimit=false
   ) {module=SpawnBurst_Instantaneous001}
   &SpawnBurst_Instantaneous(
    Spawn Count=&MakeIntFromFloat(Float=&Lerp_Float(
      A=&MakeCustomFloatFromBool(Bool=$User.bIsFrontEndPreview, True Float=0, False Float=1)
      B=&Multiply_Float(A=&RandomRangeFloat(Minimum=3, Maximum=7), B=$User.Burst)
      Alpha=$Emitter.InterpolateSpeedThresholdEmitter_01.InterpolateOverTime_Float.Moving Average
     ))
    Spawn Probability=0.7
    Use Spawn Probability=false
    UseLoopCountLimit=false
   ) {enabled=false, module=SpawnBurst_Instantaneous003}
   &SpawnRate(
    SpawnRate=&Lerp_Float(
     A=&MakeCustomFloatFromBool(Bool=$User.bIsFrontEndPreview, True Float=10, False Float=0)
     B=10
     Alpha=$Emitter.InterpolateSpeedThresholdEmitter_01.InterpolateOverTime_Float.Moving Average
    )
    Use Spawn Probability=false
   ) {enabled=false, module=SpawnRate001}
   &SpawnRate(SpawnRate=10, Use Spawn Probability=false) {enabled=false}
   &SpawnBurst_Instantaneous(
    Spawn Count=&MakeIntFromFloat(Float=&Multiply_Float(
      A=$User.Burst
      B=&MakeCustomFloatFromBool(Bool=$User.bIsFrontEndPreview, True Float=&RandomRangeFloat(Minimum=3, Maximum=10), False Float=0)
     ))
    Use Spawn Probability=false
    UseLoopCountLimit=false
   ) {enabled=false, module=SpawnBurst_Instantaneous002}
  ]
  particleSpawnScript [
   &SocketLocation(
    SamplePosOffset=false
    Socket Reader=DataInterface() {propertyValues {
      AttachComponentClass=null
      AttachComponentTag=None
      EditorPreviewAsset=/BRCosmetics/Animation/Props/Gadgets_TimberStake/Meshes/TimberStake_Gadget
      SourceActorRef=null
      SourceAsset=&TimberStake_Gadget
      SourceMode=Default
      bRequireCurrentFrameData=true
      bUpdateSocketsPerFrame=true
      FilteredSockets [mainSocket]
      ObjectParameterBinding {parameter {name=None, typeDefHandle {registeredTypeIndex=1}, varData []}}
     }}
    Socket Sampling Mode=::NewEnumerator3
    VelocitySpeedLimit=false
   )
   &InitializeParticle(
    Color=&SelectLinearColorFromArray(
     Array Sampling Mode=::NewEnumerator0
     Direct Array Index=?
     Color Selection Array=DataInterface() {propertyValues {ColorData [<rgba 0.9 0.81 0.225 1>, <rgba 1 0.45 0.5875 1>, <rgba 0.36 0.9 0.5535 1>, <rgba 0.8019 0.5 1 1>, <rgba 1 0.5127 0.3 1>]}}
     Random Seed=0
    )
    Lifetime=&MakeCustomFloatFromBool(
     Bool=$User.bIsFrontEndPreview
     True Float=&RandomRangeFloat(Minimum=0.6, Maximum=0.8)
     False Float=&RandomRangeFloat(Minimum=0.7, Maximum=1)
    )
    Mesh Scale Mode=::NewEnumerator3
    UsePositionOffset=false
    Write Material Random=true
   )
   &InitialMeshOrientation(Use Rotation Vector=false)
   &ShapeLocation(
    MaskSpawnGroup=false
    Offset=/Niagara/DynamicInputs/TypeConversions/MakeCustomPositionFromBool(Bool:NiagaraBool=$User.bIsFrontEndPreview, True Position:NiagaraPosition=<0 -40 -40>, False Position:NiagaraPosition=<0 -40 -40>)
    Offset Coordinate Space=::Local
    Offset Mode=::NewEnumerator0
    Override Seed=false
    Sphere Radius=30
    UseOwnerScale=true
   )
   &AddVelocity(
    Cone Angle=&FN_FrontEndFloatModifier(Gameplay=157.0716, Front End=45)
    Cone Axis=&FN_FrontEndVector3Modifier(Gameplay=<0 -1 133.8502>, Front End=<0 -1 0.65>)
    Inner Cone Angle=0.304
    Module.UseConeFalloff=false
    Module.UseVelDistribution=false
    Override Seed=false
    Velocity Mode=::NewEnumerator2
    Velocity Speed=&FN_FrontEndFloatModifier(
     Gameplay=&RandomRangeFloat(Minimum=350, Maximum=450)
     Front End=&RandomRangeFloat(Minimum=450, Maximum=550)
    )
   )
  ]
  particleUpdateScript [
   &ParticleState()
   &ScaleMeshSize(Scale Factor=&VectorFromFloat(Value=&FloatFromCurve(
      FloatCurve=DataInterface() {propertyValues {Curve {keys [
          {arriveTangent=4.1785, interpMode=RCIM_Cubic, leaveTangent=8.7018, tangentMode=RCTM_Break, time=0, value=0}
          {arriveTangent=0.0224, interpMode=RCIM_Cubic, leaveTangent=0.0224, tangentMode=RCTM_Break, time=0.3, value=1}
         ]}}}
      CurveIndex=$Particles.Age
      Scale Curve=&FloatFromCurve(FloatCurve=DataInterface() {propertyValues {Curve {keys [
           {interpMode=RCIM_Cubic, tangentMode=RCTM_Break, time=0.7176, value=1}
           {arriveTangent=-9.1079, interpMode=RCIM_Cubic, tangentMode=RCTM_Break, time=1, value=0}
          ]}}})
     )))
   &GravityForce(Gravity=<0 0 -700>)
   /Niagara/Modules/Update/Forces/CurlNoiseForce(Noise Frequency=60.8884, Noise Strength=298.9235, Pan Noise=false) {module=CurlNoiseForce001}
   &AerodynamicDrag(
    Aerodynamic Drag:NiagaraFloat=&FN_FrontEndFloatModifier(
     Gameplay=1
     Front End=&FloatFromCurve(FloatCurve=DataInterface() {propertyValues {Curve {keys [
          {arriveTangent=0.8669, interpMode=RCIM_Cubic, leaveTangent=0.8669, tangentMode=RCTM_User, time=0, value=0.1}
          {arriveTangent=6.2833, interpMode=RCIM_Cubic, leaveTangent=6.2833, tangentMode=RCTM_User, time=0.3, value=1}
         ]}}})
    )
    Aerodynamic Rotational Drag:NiagaraFloat=&FN_FrontEndFloatModifier(
     Gameplay=1
     Front End=&FloatFromCurve(FloatCurve=DataInterface() {propertyValues {Curve {keys [
          {arriveTangent=1.4134, interpMode=RCIM_Cubic, tangentMode=RCTM_Break, time=0, value=0}
          {arriveTangent=2.8012, interpMode=RCIM_Cubic, tangentMode=RCTM_Break, time=0.6631, value=1}
         ]}}})
    )
    Rotation Strength:NiagaraFloat=&RandomRangeFloat(Minimum=2.4141, Maximum=5.2481)
    Use Linear Drag=true
    Use Rotational Drag=true
   )
   &Collision(
    Age Colliding Particles=false
    Collision Enabled=&IsPlatformSetActive(New Platform Set=DataInterface() {propertyValues {Platforms {qualityLevelMask=-8}}})
    KillOnCollision=false
    Module.Control Rest State Via Surface Normal=false
    Module.EnableMaxCollisionCount=false
    Module.Randomize Collision Normal Vector=false
    Radius Calculation Type=::NewEnumerator1
   )
   &Collision(
    Age Colliding Particles=false
    Analytical Collision Plane Position 1:Vector3f=<0 0 -20>
    Analytical Collision Plane Space:ENiagaraCoordinateSpace=::Local
    CPU Collision Type=::NewEnumerator3
    Collision Enabled=&IsPlatformSetActive(New Platform Set=DataInterface() {propertyValues {Platforms {qualityLevelMask=-28}}})
    KillOnCollision=false
    Module.Control Rest State Via Surface Normal=false
    Module.EnableMaxCollisionCount=false
    Module.Randomize Collision Normal Vector=false
    Perform 2 Analytical Collision Queries=false
    Restitution=0
   ) {enabled=false, module=Collision001}
   &Drag(
    Drag=&FloatFromCurve(
     FloatCurve=DataInterface() {propertyValues {Curve {keys [
         {interpMode=RCIM_Cubic, tangentMode=RCTM_User, time=0.388, value=0}
         {arriveTangent=4.8344, interpMode=RCIM_Cubic, leaveTangent=4.8344, tangentMode=RCTM_User, time=1, value=1}
        ]}}}
     Scale Curve=3
    )
    Use Linear Drag=true
    Use Rotational Drag=true
   )
   &InheritSourceMovement(
    Applied Position Scale=&MakeCustomVectorFromBool(
     Bool=&SetBoolByFloatComparison(A=&VectorLength(Vector=$User.OwnerLocalVelocity), B=350)
     True Vector=<0.6 0.6 1>
     False Vector=<0 0 0>
    )
    ApplyForce=false
    ApplyPosition=true
    ApplyPositionFalloff=false
    Source Velocity:Vector3f=$User.OwnerLocalVelocity
   )
   &SolveForcesAndVelocity(Clamp Velocity=false, Limit Acceleration=false)
   &AlignSpriteToMeshOrientation(SetAlignment=true, SetFacing=true)
  ]
  rendererClasses [&NiagaraMeshRendererProperties]
  renderers [{rendererClass=&NiagaraMeshRendererProperties, rendererIndex=0}]
 }}
summary {
 systemName=NS_Emote_TimberStake_Money
 emitters [{emitterName=VBucks, rendererClasses [&NiagaraMeshRendererProperties]}]
 userVariables [
  User.bIsFrontEnd {type=NiagaraBool}
  User.bIsFrontEndPreview {type=NiagaraBool}
  User.Burst {type=NiagaraFloat, defaultValue=0, description=""}
  User.OwnerLocalVelocity {type=Vector3f, description="", defaultValue {struct=Vector3f, value=<0 0 0>}}
 ]
}
